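; DentureSay's primary Mantis attack ability, which adds air-combo handling to its technique sequences.
; Activation pushes a Mantis camera mode and caches movement and Mantis components; ending pops the camera and clears the combo.
; AirAttack techniques set the air-combo state, which applies or removes a GameplayEffect and can select a dedicated movement class.
; Root-motion warping faces and offsets from valid aim-assist targets, using hotfixable air or ground distances.
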(event Ability|EventActivateAbility
  (|CallParentFunction)
  (Camera|PushCameraMode (Variables|Camera|GetMantisCameraModeClass))
  (CallFunction|CacheComponents))

(event Ability|EventOnEndAbility
  (params (bWasCancelled :type Boolean))
  (Camera|PopCameraMode)
  (|CallParentFunction bWasCancelled)
  (CallFunction|SetIsinAirCombo false))

(event Mantis|EventOnMantisAblilityTechniqueStarted
  (params (OwningTags :type Gameplay%20Tag%20Container%20Structure) (SequenceName :type Name :default "None") (StartsSequence :type Boolean) (EndsSequence :type Boolean))
  (|CallParentFunction OwningTags SequenceName StartsSequence EndsSequence)
  (switch name SequenceName
    (:Default
      (CallFunction|SetIsinAirCombo false))
    (:AirAttack1
      _)
    (:AirAttack2
      _)
    (:AirAttack3
      _))
  (CallFunction|SetIsinAirCombo true))

(fn CacheComponents
  (Utilities|FlowControl|Sequence
    (:then_0
      (bind _asfort_movement_comp_character (Utilities|Casting|CastToFortMovementComp_Character (Pawn|GetMovementComponent (CallFunction|GetAvatarasFortPlayerPawn))))
      (Variables|Default|SetMovementComponent _asfort_movement_comp_character))
    (:then_1
      (Variables|Default|SetMantisComponent (Actor|GetComponentbyClass (CallFunction|GetAvatarasFortPlayerPawn) "/Script/MantisRuntime.FortMantisPawnComponent")))))

(fn ConfigureRootMotionWarpingAlternativeBehavior
  (params (TechniqueName :type Name :default "None"))
  (bind (_allowtranslationwarp _allowrotationwarp _snapshottargetlocation) (CallFunction|GetRootMotionOptions))
  (bind _aimassisttarget (Variables|Default|GetAimAssistTarget))
  (Utilities|FlowControl|Sequence
    (:then_0
      (CallFunction|GetAimAssistTarget)
      (if (Variables|Default|GetHasAimAssistTarget)
        (bind _returnvalue (CallFunction|IsValidTarget _aimassisttarget))
        (if _returnvalue
          (return _aimassisttarget _snapshottargetlocation "OffsetFromTargetActor" (select (Variables|Runtime|GetbIsInAirCombo) (ScalableFloat|GetValueatLevel (Variables|Hotfix|GetWarpTranslationOffsetX_InAir)) (ScalableFloat|GetValueatLevel (Variables|Hotfix|GetWarpTranslationOffsetX_InGround))) 0.0 0.0 "FaceTargetActor" "0, 0, 0"))))
    (:then_1
      (return 0 false "None" "0, 0, 0" "None" "0, 0, 0"))))

(fn GetMantisMovementClass
  (params (TechniqueData :type Fort%20Mantis%20Technique%20Data%20Structure%20%28by%20ref%29))
  (switch name (Utilities|Struct|BreakFortMantisTechniqueData TechniqueData)
    (:Default
      _)
    (:AirAttack1
      (return (Variables|Default|GetMME_AirCombo)))
    (:AirAttack2
      (if (Variables|Default|GetHasAimAssistTarget)
        _
        (else
          _)))
    (:AirAttack3
      _))
  (bind _returnvalue (|CallParentFunction TechniqueData))
  (return _returnvalue))

(fn SetIsInAirCombo
  (params (bInIsInAirCombo :type Boolean))
  (if (!= bInIsInAirCombo (Variables|Runtime|GetbIsInAirCombo))
    (bind _output_get (Variables|Runtime|SetbIsInAirCombo bInIsInAirCombo))
    ;[0.06408 0.15 0.1268 0.5] Update GE
    (if _output_get
      (bind _returnvalue (Ability|ApplyGameplayEffectToOwner (Variables|Default|GetGE_AirComboActive)))
      (Variables|Runtime|SetAirComboActiveHandle _returnvalue)
      (else
        (Ability|RemoveGameplayEffectFromOwnerWithHandle (Variables|Runtime|GetAirComboActiveHandle))))))
